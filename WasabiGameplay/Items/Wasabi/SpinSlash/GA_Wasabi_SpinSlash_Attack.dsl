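; Wasabi's Spin Slash attack ability, which launches the avatar into a forward jump and strikes on montage notifies.
; A root-motion move-to force blends tunable distance, height and end speed with clamped, partially preserved incoming velocity.
; Landing stops movement, ends that task and applies a move-speed effect until ability end; knockback ends the ability.
; Notifies step through hit gameplay cues, and the major-hit notify swaps the targeting tag and runs targeting.
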(event Custom|BindOnKnockedBack
  ;[1 1 1] On Knocked Back
  (bind _asynctaskproxy (Ability|Tasks|WaitGameplayTagAdd (Variables|Default|GetTag_KnockedBack))
    (:then
      (Variables|Default|SetTask_OnKnockedBack _asynctaskproxy))
    (:Added
      (Ability|EndAbility))))

(event Ability|EventActivateAbility
  (bind (_initialforwardvector_x _initialforwardvector_y) (Math|Vector|BreakVector (Variables|GetInitialForwardVector_X)))
  (bind (_initialvelocityclamped_x _initialvelocityclamped_y _initialvelocityclamped_z) (Math|Vector|BreakVector (Variables|GetInitialVelocityClamped_X)))
  (bind (_initialvelocityclamped_x_2 _initialvelocityclamped_y_2 _initialvelocityclamped_z_2) (Math|Vector|BreakVector (Variables|GetInitialVelocityClamped_X)))
  ;[1 1 1] On Activate Ability jump the player.
  (CallFunction|BindonKnockedBack)
  (Variables|Default|SetInitialForwardVector (Transformation|GetActorForwardVector (Ability|GetAvatarActorfromActorInfo)))
  (Variables|Default|SetInitialVelocityClamped (Math|Vector|VectorClampSizeMax (Transformation|GetVelocity (Ability|GetAvatarActorfromActorInfo)) (select (>= (ScalableFloat|GetValueatLevel (Variables|Balance|GetMaxSpeedToMaintain)) 0.0) (ScalableFloat|GetValueatLevel (Variables|Balance|GetMaxSpeedToMaintain)) (Math|Vector|VectorLength (Transformation|GetVelocity (Ability|GetAvatarActorfromActorInfo))))))
  (bind _asynctaskproxy (Ability|Tasks|ApplyRootMotionMoveToForce "None" (+ (Transformation|GetActorLocation (Ability|GetAvatarActorfromActorInfo)) (Math|Vector|vector+vector (* (Math|Vector|Normalize (Math|Vector|MakeVector _initialforwardvector_x _initialforwardvector_y)) (ScalableFloat|GetValueatLevel (Variables|Balance|GetJumpAddedDistance))) (Math|Vector|MakeVector (* _initialvelocityclamped_x (* (ScalableFloat|GetValueatLevel (Variables|Balance|GetJumpDuration)) (ScalableFloat|GetValueatLevel (Variables|Balance|GetPercentForwardVelocityToMaintain)))) (* _initialvelocityclamped_y (* (ScalableFloat|GetValueatLevel (Variables|Balance|GetJumpDuration)) (ScalableFloat|GetValueatLevel (Variables|Balance|GetPercentForwardVelocityToMaintain)))) (+ (* _initialvelocityclamped_z (select (Variables|Default|GetFloorMaintainedZVelocityAtZero) (Math|Float|Max(Float) (* (ScalableFloat|GetValueatLevel (Variables|Balance|GetJumpDuration)) (ScalableFloat|GetValueatLevel (Variables|Balance|GetPercentVerticalVelocityToMaintain)))) (* (ScalableFloat|GetValueatLevel (Variables|Balance|GetJumpDuration)) (ScalableFloat|GetValueatLevel (Variables|Balance|GetPercentVerticalVelocityToMaintain))))) (ScalableFloat|GetValueatLevel (Variables|Balance|GetJumpAddedHeight)))))) (ScalableFloat|GetValueatLevel (Variables|Balance|GetJumpDuration)) false "MOVE_None" false 0 "SetVelocity" (Math|Vector|vector+vector (* (Variables|Default|GetInitialForwardVector) (ScalableFloat|GetValueatLevel (Variables|Balance|GetEndAddedSpeed))) (Math|Vector|MakeVector (* _initialvelocityclamped_x_2 (ScalableFloat|GetValueatLevel (Variables|Balance|GetPercentForwardVelocityToMaintain))) (* _initialvelocityclamped_y_2 (ScalableFloat|GetValueatLevel (Variables|Balance|GetPercentForwardVelocityToMaintain))) (select (Variables|Default|GetFloorMaintainedZVelocityAtZero) (Math|Float|Max(Float) (* _initialvelocityclamped_z_2 (ScalableFloat|GetValueatLevel (Variables|Balance|GetPercentVerticalVelocityToMaintain)))) (* _initialvelocityclamped_z_2 (ScalableFloat|GetValueatLevel (Variables|Balance|GetPercentVerticalVelocityToMaintain))))))))
  (Variables|Default|SetTask_RootMotion _asynctaskproxy)
  (|CallParentFunction)
  (EventDispatchers|BindEventtoOnPawnLanded (Variables|Default|GetPlayerPawn) (EventDispatchers|CreateEvent :!event Custom|Landed)))

(event Custom|Landed
  (params (Hit :type Hit%20Result%20Structure%20%28by%20ref%29))
  ;[1 1 1] On Landed
  (Components|Movement|StopMovementImmediately (Variables|/Script/Engine.Character|GetCharacterMovement (Variables|Default|GetPlayerPawn)))
  (bind _returnvalue (Ability|ApplyGameplayEffectToOwner (Variables|Default|GetGE_MoveSpeedOnLanded)))
  (Variables|Default|SetHandle_GE_MoveSpeedOnLanded _returnvalue)
  (GameplayTasks|EndTask (Variables|Default|GetTask_RootMotion)))

(event Ability|EventOnEndAbility
  (params (bWasCancelled :type Boolean))
  ;[1 1 1] On End Ability - Unbind Land Events
  (|CallParentFunction bWasCancelled)
  (EventDispatchers|UnbindEventfromOnPawnLanded (Variables|Default|GetPlayerPawn) (EventDispatchers|CreateEvent :!event Custom|Landed))
  (Ability|RemoveGameplayEffectFromOwnerWithHandle (Variables|Default|GetHandle_GE_MoveSpeedOnLanded))
  (GameplayTasks|EndTask (Variables|Default|GetTask_OnKnockedBack))
  (GameplayTasks|EndTask (Variables|Default|GetTask_RootMotion)))

(event EventCastMontageNotifyStart
  (params (NotifyTag :type Gameplay%20Tag%20Structure))
  (bind _tag_majorhit (Variables|Default|GetTag_MajorHit))
  (bind _gcindex (Variables|Default|GetGCIndex))
  ;[1 1 1] When we hit a trigger set the correct GC to play and if it is the big hit change the targeting application tag to the big hit one.
  (if (< (Variables|Default|GetGCIndex) (Utilities|Array|Length (Variables|Default|GetGC_Hits)))
    (Variables|PowerGameplayCues|SetGC_Targeting (Utilities|Array|Get(aref) (Variables|Default|GetGC_Hits) _gcindex))
    (Math|Integer|IncrementInt _gcindex
      (:%20%20
        _))
    (else
      _))
  (|CallParentFunction NotifyTag)
  (if (GameplayTags|Equal(GameplayTag) _tag_majorhit NotifyTag)
    (Variables|PowerTargetApply|SetTag_TargetingApplication _tag_majorhit)
    (CallFunction|Target)))

(fn InitAbility
  ;[1 1 1] Reset target application tag and also set index to 0 for the GC
  (Variables|PowerTargetApply|SetTag_TargetingApplication "(TagName=\"Targeting.ApplicationTag.0\")")
  (Variables|Default|SetGCIndex 0)
  (|CallParentFunction))
